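# Dataflow template that turns a StaticMesh variable into a skinned skeletal mesh with an automatically generated skeleton.
# The static mesh converts to a mesh that feeds medial-axis sphere sampling, skeleton simplification and skinning.
# Skinning derives bones from the simplified medial skeleton and binds by direct distance before the skeletal mesh terminal.
# An asset-path branch with an SK_ prefix rule supplies only the skeleton path; its asset input is unconnected.
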Dataflow /Dataflow/Templates/SKM/DF_SKM_FromMesh_Template
name=DF_SKM_FromMesh_Template
connections [
 {
  fromNode=MeshMedialSkeletonSampling
  fromPin=MedialSkeleton
  fromPinGuid=#1
  toNode=SimplifyMedialSkeleton_0
  toPin=MedialSkeleton
  toPinGuid=#2
 }
 {
  fromNode=SkinMeshViaMedialSkeleton_0
  fromPin=Mesh
  fromPinGuid=#3
  toNode=MeshToSkeletalMeshTerminal_0
  toPin=Mesh
  toPinGuid=#4
 }
 {
  fromNode=SimplifyMedialSkeleton_0
  fromPin=MedialSkeleton
  fromPinGuid=#5
  toNode=SkinMeshViaMedialSkeleton_0
  toPin=MedialSkeleton
  toPinGuid=#6
 }
 {
  fromNode=StaticMesh
  fromPin=Value
  fromPinGuid=#7
  toNode=StaticMeshToMesh
  toPin=StaticMesh
  toPinGuid=#8
 }
 {
  fromNode=StaticMeshToMesh
  fromPin=Mesh
  fromPinGuid=#9
  toNode=MeshMedialSkeletonSampling
  toPin=TargetMesh
  toPinGuid=#10
 }
 {
  fromNode=StaticMeshToMesh
  fromPin=Mesh
  fromPinGuid=#9
  toNode=ReRouteNode_v1
  toPin=Value
  toPinGuid=#11
 }
 {
  fromNode=ReRouteNode_v1_4
  fromPin=Value
  fromPinGuid=#12
  toNode=SimplifyMedialSkeleton_0
  toPin=TargetMesh
  toPinGuid=#13
 }
 {
  fromNode=ReRouteNode_v1_5
  fromPin=Value
  fromPinGuid=#14
  toNode=SkinMeshViaMedialSkeleton_0
  toPin=Mesh
  toPinGuid=#15
 }
 {
  fromNode=ReRouteNode_v1_4
  fromPin=Value
  fromPinGuid=#12
  toNode=ReRouteNode_v1_5
  toPin=Value
  toPinGuid=#16
 }
 {
  fromNode=PathReplace
  fromPin=Path
  fromPinGuid=#17
  toNode=MeshToSkeletalMeshTerminal_0
  toPin=SkeletonAssetPath
  toPinGuid=#18
 }
 {
  fromNode=GetAssetPath
  fromPin=Path
  fromPinGuid=#19
  toNode=PathReplace
  toPin=Path
  toPinGuid=#20
 }
 {
  fromNode=ReRouteNode_v1
  fromPin=Value
  fromPinGuid=#21
  toNode=ReRouteNode_v1_4
  toPin=Value
  toPinGuid=#22
 }
]
nodes [
 StaticMesh {
  type=FGetDataflowVariableNode
  guid=#23
  outputPins [Value {type="TObjectPtr<UStaticMesh>"}]
  properties {VariableName=StaticMesh, VariablePropertyBag=(Value=/Engine/Transient.PropertyBag_2f6e171ba8c10008(StaticMesh=None))}
 }
 GetAssetPath {
  type=FDataflowGetAssetPathNode
  guid=#24
  inputPins [Asset {type=FDataflowUObjectConvertibleTypes}]
  outputPins [Path {type=FString}]
  properties {Asset=()}
 }
 PathReplace {
  type=FDataflowPathReplaceNode
  guid=#25
  inputPins [Path {type=FString}]
  outputPins [Path {type=FString}]
  properties {NameRules=((Method=Beginning,From="*",To="SK_",bInsertIfNoFound=True)), PathRules=""}
 }
 StaticMeshToMesh {
  type=FStaticMeshToMeshDataflowNode_v2
  guid=#26
  inputPins [StaticMesh {type="TObjectPtr<UStaticMesh>"}]
  outputPins [Mesh {type="TObjectPtr<UDataflowMesh>"}]
  properties {LODLevel=0, StaticMesh=None, bUseHiRes=False}
 }
 MeshMedialSkeletonSampling {
  type=FMeshMedialSkeletonSamplingDataflowNode_v2
  guid=#27
  inputPins [
   TargetMesh {type="TObjectPtr<UDataflowMesh>"}
   MaxSpheres {type=int32}
   MinClusterErrorToSplit {type=float}
   MinClusterSizeToKeep {type=int32}
   bSplitClustersIfEdgesIntersectSurface {type=bool}
   ErrorScaleNearEdgeSurfaceIntersection {type=float}
   PosErrorWt {type=float}
  ]
  outputPins [
   MedialSkeleton {type=FDataflowMedialSkeleton}
   MedialSpheres {type="TArray<FSphere>"}
  ]
  properties {
   ErrorScaleNearEdgeSurfaceIntersection=10
   MaxSpheres=1000
   MinClusterErrorToSplit=10
   MinClusterSizeToKeep=4
   PosErrorWt=0.2
   bSplitClustersIfEdgesIntersectSurface=True
  }
 }
 SimplifyMedialSkeleton_0 {
  type=FSimplifyMedialSkeletonDataflowNode_v2
  guid=#28
  inputPins [
   MedialSkeleton {type=FDataflowMedialSkeleton}
   TargetMesh {type="TObjectPtr<UDataflowMesh>"}
   MinSpheres {type=int32}
   QEMErrorThresholdSqrt {type=float}
   EdgeLengthThreshold {type=float}
   SphereRadiusThreshold {type=float}
   SphereOverlapThreshold {type=float}
   bOnlySimplifySurfaces {type=bool}
   ClusterSkeletonDistanceWt {type=float}
   bPreventEdgeSurfaceIntersections {type=bool}
   bSplitThinTriEdges {type=bool}
   SplitThinTriEdgeAngleThresholdDeg {type=float}
   SplitThinTriEdgePosErrorWt {type=float}
  ]
  outputPins [
   MedialSkeleton {type=FDataflowMedialSkeleton}
   MedialSpheres {type="TArray<FSphere>"}
  ]
  properties {
   ClusterSkeletonDistanceWt=0.5
   EdgeLengthThreshold=0
   MinSpheres=1
   QEMErrorThresholdSqrt=0
   SphereOverlapThreshold=0
   SphereRadiusThreshold=0
   SplitThinTriEdgeAngleThresholdDeg=120
   SplitThinTriEdgePosErrorWt=0.2
   bOnlySimplifySurfaces=True
   bPreventEdgeSurfaceIntersections=True
   bSplitThinTriEdges=True
  }
 }
 SkinMeshViaMedialSkeleton_0 {
  type=FSkinMeshViaMedialSkeleton_v2
  guid=#29
  inputPins [
   Mesh {type="TObjectPtr<UDataflowMesh>"}
   MedialSkeleton {type=FDataflowMedialSkeleton}
   RootSelectionPoint {type=FVector}
   RootSelectionDirection {type=FVector}
   RootClusterIndex {type=int32}
   bAddCustomAnimationRoot {type=bool}
   CustomAnimationRootPosition {type=FVector}
   Stiffness {type=float}
   MaxInfluences {type=int32}
   VoxelResolution {type=int32}
   ClusterNeighborSearchRange {type=int32}
  ]
  outputPins [Mesh {type="TObjectPtr<UDataflowMesh>"}]
  properties {
   BindMethod=DirectDistance
   ClusterNeighborSearchRange=1
   CustomAnimationRootPosition=(X=0,Y=0,Z=0)
   EdgeWeightMethod=ArrayOrder
   MaxInfluences=5
   MergeDisconnectedMethod=ConnectClosestBones
   RootClusterIndex=-1
   RootSelectionDirection=(X=0,Y=0,Z=-1)
   RootSelectionPoint=(X=0,Y=0,Z=0)
   SelectRootMethod=ArrayOrder
   Stiffness=0.2
   VoxelResolution=256
   bAddCustomAnimationRoot=False
  }
 }
 ReRouteNode_v1_4 {
  type=FDataflowReRouteNode
  guid=#30
  inputPins [Value {type="TObjectPtr<UDataflowMesh>"}]
  outputPins [Value {type="TObjectPtr<UDataflowMesh>"}]
 }
 ReRouteNode_v1_5 {
  type=FDataflowReRouteNode
  guid=#31
  inputPins [Value {type="TObjectPtr<UDataflowMesh>"}]
  outputPins [Value {type="TObjectPtr<UDataflowMesh>"}]
 }
 MeshToSkeletalMeshTerminal_0 {
  type=FMeshToSkeletalMeshTerminalNode_v2
  guid=#32
  inputPins [
   Mesh {type="TObjectPtr<UDataflowMesh>"}
   SkeletalMeshAssetPath {type=FString}
   SkeletonAssetPath {type=FString}
  ]
  outputPins [
   SkeletalMeshAsset {type="TObjectPtr<USkeletalMesh>"}
   SkeletonAsset {type="TObjectPtr<USkeleton>"}
  ]
  properties {SkeletalMeshAssetPath="", SkeletonAssetPath=""}
 }
 ReRouteNode_v1 {
  type=FDataflowReRouteNode
  guid=#33
  inputPins [Value {type="TObjectPtr<UDataflowMesh>"}]
  outputPins [Value {type="TObjectPtr<UDataflowMesh>"}]
 }
]
