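# Dataflow template that converts the SourceMesh static mesh variable into a skeletal mesh and skeleton, splitting the mesh into components.
# StaticMeshToCollection splits components, then passes the collection and materials to CollectionToSkeletalMesh.
# The result becomes a Dataflow mesh via SkeletalMeshToMesh and feeds the skeletal mesh terminal.
# An asset path with an SK_ name-prefix rule drives only the terminal's skeleton path; its skeletal mesh path is empty.
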Dataflow /Dataflow/Templates/SKM/DF_SKM_MeshSplit_Template
name=DF_SKM_MeshSplit_Template
connections [
 {
  fromNode=StaticMeshToCollection
  fromPin=Collection
  fromPinGuid=#1
  toNode=CollectionToSkeletalMesh
  toPin=Collection
  toPinGuid=#2
 }
 {
  fromNode=SourceMesh
  fromPin=Value
  fromPinGuid=#3
  toNode=StaticMeshToCollection
  toPin=StaticMesh
  toPinGuid=#4
 }
 {
  fromNode=CollectionToSkeletalMesh
  fromPin=SkeletalMesh
  fromPinGuid=#5
  toNode=SkeletalMeshToMesh
  toPin=SkeletalMesh
  toPinGuid=#6
 }
 {
  fromNode=SkeletalMeshToMesh
  fromPin=Mesh
  fromPinGuid=#7
  toNode=MeshToSkeletalMeshTerminal
  toPin=Mesh
  toPinGuid=#8
 }
 {
  fromNode=GetAssetPath
  fromPin=Path
  fromPinGuid=#9
  toNode=PathReplace
  toPin=Path
  toPinGuid=#10
 }
 {
  fromNode=PathReplace
  fromPin=Path
  fromPinGuid=#11
  toNode=MeshToSkeletalMeshTerminal
  toPin=SkeletonAssetPath
  toPinGuid=#12
 }
 {
  fromNode=StaticMeshToCollection
  fromPin=Materials
  fromPinGuid=#13
  toNode=CollectionToSkeletalMesh
  toPin=Materials
  toPinGuid=#14
 }
]
nodes [
 SourceMesh {
  type=FGetDataflowVariableNode
  guid=#15
  outputPins [Value {type="TObjectPtr<UStaticMesh>"}]
  properties {VariableName=SourceMesh, VariablePropertyBag=(Value=/Engine/Transient.PropertyBag_d52a8e2be636b1f6(SourceMesh=None))}
 }
 StaticMeshToCollection {
  type=FStaticMeshToCollectionDataflowNode_v2
  guid=#16
  inputPins [
   StaticMesh {type="TObjectPtr<UStaticMesh>"}
   MeshTransform {type=FTransform}
  ]
  outputPins [
   Collection {type=FManagedArrayCollection}
   Materials {type="TArray<TObjectPtr<UMaterialInterface>>"}
   InstancedMeshes {type="TArray<FGeometryCollectionAutoInstanceMesh>"}
   RootProxyMeshes {type="TArray<FDataflowRootProxyMesh>"}
  ]
  properties {
   MeshTransform=(Rotation=(X=0,Y=0,Z=0,W=1),Translation=(X=0,Y=0,Z=0),Scale3D=(X=1,Y=1,Z=1))
   StaticMesh=None
   bSetInternalFromMaterialIndex=True
   bSplitComponents=True
  }
 }
 CollectionToSkeletalMesh {
  type=FCollectionToSkeletalMeshDataflowNode
  guid=#17
  inputPins [
   Collection {type=FManagedArrayCollection}
   Materials {type="TArray<TObjectPtr<UMaterialInterface>>"}
  ]
  outputPins [
   SkeletalMesh {type="TObjectPtr<USkeletalMesh>"}
   Skeleton {type="TObjectPtr<USkeleton>"}
  ]
 }
 SkeletalMeshToMesh {
  type=FSkeletalMeshToMeshDataflowNode_v2
  guid=#18
  inputPins [SkeletalMesh {type="TObjectPtr<USkeletalMesh>"}]
  outputPins [Mesh {type="TObjectPtr<UDataflowMesh>"}]
  properties {LODLevel=0, SkeletalMesh=None, bRecordImportedVertices=False, bUseMeshDescription=True}
 }
 MeshToSkeletalMeshTerminal {
  type=FMeshToSkeletalMeshTerminalNode_v2
  guid=#19
  inputPins [
   Mesh {type="TObjectPtr<UDataflowMesh>"}
   SkeletalMeshAssetPath {type=FString}
   SkeletonAssetPath {type=FString}
  ]
  outputPins [
   SkeletalMeshAsset {type="TObjectPtr<USkeletalMesh>"}
   SkeletonAsset {type="TObjectPtr<USkeleton>"}
  ]
  properties {SkeletalMeshAssetPath="", SkeletonAssetPath=""}
 }
 GetAssetPath {
  type=FDataflowGetAssetPathNode
  guid=#20
  inputPins [Asset {type=FDataflowUObjectConvertibleTypes}]
  outputPins [Path {type=FString}]
  properties {Asset=()}
 }
 PathReplace {
  type=FDataflowPathReplaceNode
  guid=#21
  inputPins [Path {type=FString}]
  outputPins [Path {type=FString}]
  properties {NameRules=((Method=Beginning,From="*",To="SK_",bInsertIfNoFound=True)), PathRules=""}
 }
]
