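# Dataflow template that rebuilds a skeletal mesh from new static mesh geometry while keeping the existing skeletal mesh's data.
# The SourceStaticMesh and SkeletalMeshTarget variables are each converted to meshes at LOD 0.
# TransferMeshAttributes copies skeleton, skin weights, morph targets, triangle labels and polygroups from the skeletal mesh onto the static geometry by closest point on surface.
# The result feeds a skeletal mesh terminal, whose asset path is left empty.
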Dataflow /Dataflow/Templates/SKM/DF_SKM_ReimportFromMesh_Template
name=DF_SKM_ReimportFromMesh_Template
connections [
 {
  fromNode=TransferMeshAttributes
  fromPin=Mesh
  fromPinGuid=#1
  toNode=SkeletalMeshTerminal
  toPin=Mesh
  toPinGuid=#2
 }
 {
  fromNode=SourceStaticMesh_0
  fromPin=Value
  fromPinGuid=#3
  toNode=StaticMeshToMesh_v2
  toPin=StaticMesh
  toPinGuid=#4
 }
 {
  fromNode=SkeletalMeshTarget_0
  fromPin=Value
  fromPinGuid=#5
  toNode=SkeletalMeshToMesh
  toPin=SkeletalMesh
  toPinGuid=#6
 }
 {
  fromNode=StaticMeshToMesh_v2
  fromPin=Mesh
  fromPinGuid=#7
  toNode=TransferMeshAttributes
  toPin=Mesh
  toPinGuid=#8
 }
 {
  fromNode=SkeletalMeshToMesh
  fromPin=Mesh
  fromPinGuid=#9
  toNode=TransferMeshAttributes
  toPin=SourceMesh
  toPinGuid=#10
 }
]
nodes [
 SkeletalMeshTerminal {
  type=FSkeletalMeshTerminalNode
  guid=#11
  inputPins [
   Mesh {type="TObjectPtr<UDataflowMesh>"}
   AssetPath {type=FString}
  ]
  outputPins [SkeletalMeshAsset {type="TObjectPtr<USkeletalMesh>"}]
  properties {AssetPath=""}
 }
 TransferMeshAttributes {
  type=FTransferMeshAttributesDataflowNode
  guid=#12
  inputPins [
   Mesh {type="TObjectPtr<UDataflowMesh>"}
   SourceMesh {type="TObjectPtr<UDataflowMesh>"}
  ]
  outputPins [Mesh {type="TObjectPtr<UDataflowMesh>"}]
  properties {
   All=()
   AttributeProxies=(/Script/DataflowNodes.DataflowSkeletonProxy(),/Script/DataflowNodes.DataflowSkinWeightsProxy(SourceType=All,Source="",Destination=""),/Script/DataflowNodes.DataflowMorphTargetProxy(SourceType=All,Source="",Destination=""),/Script/DataflowNodes.DataflowTriangleLabelsProxy(SourceType=All,Source="",Destination=""),/Script/DataflowNodes.DataflowPolygroupProxy(SourceType=All,Source="",Destination=""))
   None=()
   NormalThreshold=30
   SearchRadius=0.05
   SourceMesh=None
   TransferMethod=ClosestPointOnSurface
   bLayeredMeshSupport=False
   bUseIntrinsicLaplacian=False
  }
 }
 SourceStaticMesh_0 {
  type=FGetDataflowVariableNode
  guid=#13
  outputPins [Value {type="TObjectPtr<UStaticMesh>"}]
  properties {VariableName=SourceStaticMesh, VariablePropertyBag=(Value=/Engine/Transient.PropertyBag_3bcd36fb08845316(SourceStaticMesh=None))}
 }
 StaticMeshToMesh_v2 {
  type=FStaticMeshToMeshDataflowNode_v2
  guid=#14
  inputPins [StaticMesh {type="TObjectPtr<UStaticMesh>"}]
  outputPins [Mesh {type="TObjectPtr<UDataflowMesh>"}]
  properties {LODLevel=0, StaticMesh=None, bUseHiRes=False}
 }
 SkeletalMeshToMesh {
  type=FSkeletalMeshToMeshDataflowNode_v2
  guid=#15
  inputPins [SkeletalMesh {type="TObjectPtr<USkeletalMesh>"}]
  outputPins [Mesh {type="TObjectPtr<UDataflowMesh>"}]
  properties {LODLevel=0, SkeletalMesh=None, bRecordImportedVertices=False, bUseMeshDescription=True}
 }
 SkeletalMeshTarget_0 {
  type=FGetDataflowVariableNode
  guid=#16
  outputPins [Value {type="TObjectPtr<USkeletalMesh>"}]
  properties {VariableName=SkeletalMeshTarget, VariablePropertyBag=(Value=/Engine/Transient.PropertyBag_3bca96200c71b29b(SkeletalMeshTarget=None))}
 }
]
