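# Dataflow template graph that builds a Catmull-Clark subdivided skeletal mesh from a source skeletal mesh.
# The SourceSkeletalMesh variable, unset by default, feeds a conversion of the base LOD to a Dataflow mesh, recording imported vertices.
# A single Catmull-Clark subdivision pass, which autogenerates polygroups if missing, refines that mesh.
# A skeletal mesh terminal receives the result as the output asset, with its asset path left blank.
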Dataflow /Dataflow/Templates/SKM/DF_SKM_SubD_Template
name=DF_SKM_SubD_Template
connections [
 {
  fromNode=SubdivideMeshCatmullClark
  fromPin=Mesh
  fromPinGuid=#1
  toNode=SkeletalMeshTerminal
  toPin=Mesh
  toPinGuid=#2
 }
 {
  fromNode=SkeletalMeshToMesh
  fromPin=Mesh
  fromPinGuid=#3
  toNode=SubdivideMeshCatmullClark
  toPin=Mesh
  toPinGuid=#4
 }
 {
  fromNode=SourceSkeletalMesh
  fromPin=Value
  fromPinGuid=#5
  toNode=SkeletalMeshToMesh
  toPin=SkeletalMesh
  toPinGuid=#6
 }
]
nodes [
 SubdivideMeshCatmullClark {
  type=FSubdivideMeshCatmullClarkDataflowNode
  guid=#7
  inputPins [
   Mesh {type="TObjectPtr<UDataflowMesh>"}
   Subdivisions {type=int32}
  ]
  outputPins [Mesh {type="TObjectPtr<UDataflowMesh>"}]
  properties {AutogenPolygroups=IfMissing, ExtendedLayerIndex=0, Subdivisions=1, bDefaultLayer=True}
 }
 SkeletalMeshTerminal {
  type=FSkeletalMeshTerminalNode
  guid=#8
  inputPins [
   Mesh {type="TObjectPtr<UDataflowMesh>"}
   AssetPath {type=FString}
  ]
  outputPins [SkeletalMeshAsset {type="TObjectPtr<USkeletalMesh>"}]
  properties {AssetPath=""}
 }
 SourceSkeletalMesh {
  type=FGetDataflowVariableNode
  guid=#9
  outputPins [Value {type="TObjectPtr<USkeletalMesh>"}]
  properties {VariableName=SourceSkeletalMesh, VariablePropertyBag=(Value=/Engine/Transient.PropertyBag_2da68f18c43d8915(SourceSkeletalMesh=None))}
 }
 SkeletalMeshToMesh {
  type=FSkeletalMeshToMeshDataflowNode_v2
  guid=#10
  inputPins [SkeletalMesh {type="TObjectPtr<USkeletalMesh>"}]
  outputPins [Mesh {type="TObjectPtr<UDataflowMesh>"}]
  properties {LODLevel=0, SkeletalMesh=None, bRecordImportedVertices=True, bUseMeshDescription=False}
 }
]
