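# Editor utility Blueprint that reimports selected assets, presumably skeletal meshes, from a given static mesh through Dataflow.
# Its Reimport From Mesh function takes a SourceMesh and loops over the selected assets.
# Each asset gets Dataflow variable overrides: SourceStaticMesh is the input mesh, and SkeletalMeshTarget is the asset itself.
# It then regenerates the asset from its Dataflow, including dependent assets.
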EditorUtilityBlueprint /Dataflow/Templates/SKM/EUB_Reimport_FromMesh
graphs [
 EventGraph {links [], nodes []}
 Reimport From Mesh {nodes [
   n1 = Reimport From Mesh!() -> (then, SourceMesh:Static Mesh Object Reference) {is=title K2Node_FunctionEntry}
   > n2 = GetSelectedAssets()
   > n3 = For Each Loop(Array:Array of Object References=@n2) -> (Array Element:Object Reference)
   > OverrideDataflowVariableObject!(execute, Asset:Object Reference=@n3, VariableName:Name=SourceStaticMesh, VariableValue:Object Reference=@n1) -> (then, ReturnValue:bool)
   > OverrideDataflowVariableObject(Asset=@n3, VariableName=SkeletalMeshTarget, VariableValue=@n3)
   > RegenerateAssetFromDataflow!(execute, AssetToRegenerate:Object Reference=@n3, bRegenerateDependentAssets:bool=true) -> (then, ReturnValue:bool)
  ]}
]
variables []
